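# MF_FlipbookTexture is a material function that outputs the RGB of one frame from a flipbook texture, chosen by an index input.
# It appends the column and row count inputs into a sub-image grid and passes it with the texture, UVs and index to MF_FlipbookSubUV.
# Every input has a preview default, including a 4x4 grid, a placeholder texture and standard texture coordinates.
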MaterialFunction /Game/Global/Materials/Character_Parent_2023/VideoFlipbook_Parent_2025/MaterialFunctions/MF_FlipbookTexture
graph {nodes [
  n1 = TextureObject() {texture=&T_BlacksRGB}
  n2 = TextureCoordinate()
  n3 = FunctionInput() {bUsePreviewValueAsDefault=true, inputName=Flipbook Column Count, inputType=FunctionInput_Scalar, sortPriority=1, previewValue=<4 0 0 1>}
  n4 = FunctionInput() {bUsePreviewValueAsDefault=true, inputName=Flipbook Row Count, inputType=FunctionInput_Scalar, sortPriority=2, previewValue=<4 0 0 1>}
  FunctionOutput(None=MaterialFunctionCall!(
    Texture=FunctionInput(Preview=@n1) {bUsePreviewValueAsDefault=true, inputName=Flipbook Texture, inputType=FunctionInput_Texture2D}
    UVs=FunctionInput(Preview=@n2) {bUsePreviewValueAsDefault=true, inputName=UVs, inputType=FunctionInput_Vector2, sortPriority=3}
    SubImages=AppendVector(A=@n3, B=@n4)
    Frame=FunctionInput() {bUsePreviewValueAsDefault=true, inputName=Flipbook Index, inputType=FunctionInput_Scalar, sortPriority=-10}
   ) {materialFunction=&/MF_FlipbookSubUV}.RGB) {outputName=Frame}
 ]}
