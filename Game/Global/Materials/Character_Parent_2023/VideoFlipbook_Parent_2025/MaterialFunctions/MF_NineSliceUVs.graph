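# Material function that remaps UVs for nine-slice texturing, splitting them into edge and center bands sized by EdgeRatio.
# Nested If nodes stretch each band into its own unit range, yielding a 0–3 deformed UV, whose floor gives the 2D slice index.
# That UV, divided by three, samples an evenly sliced 3x3 texture, using original-UV derivatives for mip selection.
# Outputs are deformed UV, slice index and sampled RGB; the EdgeWidth parameter is unconnected.
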MaterialFunction /Game/Global/Materials/Character_Parent_2023/VideoFlipbook_Parent_2025/MaterialFunctions/MF_NineSliceUVs
graph {nodes [
  n1 = FunctionInput(Preview=TextureCoordinate()) {bUsePreviewValueAsDefault=true, inputName=UVs, inputType=FunctionInput_Vector2}
  n2 = FunctionInput() {bUsePreviewValueAsDefault=true, inputName=EdgeRatio, inputType=FunctionInput_Scalar, previewValue=<0.25 0 0 1>}
  n3 = Divide(A=@n2)
  n4 = Divide(A=Subtract(A=@n1, B=@n3), B=OneMinus(None=@n3))
  n5 = OneMinus(None=Divide(A=@n3, B=OneMinus(None=@n3)))
  n6 = Add(A=Divide(A=Subtract(A=@n4, B=@n5), B=OneMinus(None=@n5)), B=Constant2Vector() {g=1, r=1})
  n7 = Divide(A=@n4, B=@n5)
  n8 = If(A=@n1, B=@n3, "A > B"=Add(A=If(A=@n4, B=@n5, "A > B"=@n6, "A < B"=@n7), B=Constant2Vector() {g=1, r=1}), "A < B"=Divide(A=@n1, B=@n3))
  FunctionOutput(None=@n8) {outputName=Deformed UV}
  FunctionOutput(None=Floor(None=@n8) {desc=Tile coords}) {outputName="Slice Index (2D)", sortPriority=10}
  n9 = TextureSample!(
   UVs=Divide(A=@n8) {constB=3}
   Tex=FunctionInput(Preview=TextureObject() {texture=&/T_9SliceExample}) {inputName="Texture (even slices 3x3)", inputType=FunctionInput_Texture2D}
   "DDX(UVs)"=DDX(Value=@n1)
   "DDY(UVs)"=DDY(Value=@n1)
   Apply View MipBias
  ) {
   mipValueMode=TMVM_Derivative
   samplerSource=SSM_Wrap_WorldGroupSettings
   texture=&T_9SliceExample
  }
  FunctionOutput(None=@n9.RGB) {outputName=Sliced Texture}
  ScalarParameter() {defaultValue=0.4, parameterName=EdgeWidth, sliderMax=1}
 ]}
