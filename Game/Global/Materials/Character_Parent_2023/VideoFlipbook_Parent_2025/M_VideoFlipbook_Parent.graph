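# Translucent unlit, two-sided parent material that plays video-like flipbooks inside an optional glowing, glitching border.
# Up to three flipbook textures, each with its own grid and frame count, are indexed by FPS and a start time, then combined through Step blends and static switches.
# Nine-slice UVs sample frame, content, inner-glow and glitch masks.
# Time-jittered rectangle textures drive the glitches, and the border adds to the flipbook emissive.
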Material /Game/Global/Materials/Character_Parent_2023/VideoFlipbook_Parent_2025/M_VideoFlipbook_Parent
graph {
 properties {
  bAutomaticallySetUsageInEditor=false
  bHasPixelAnimation=true
  bOutputTranslucentVelocity=true
  blendMode=BLEND_Translucent
  shadingModel=MSM_Unlit
  twoSided=true
  usedShadingModels=Unlit
 }
 nodes [
  Constant() {r=1}
  n1 = VectorParameter() {group=Frame, parameterName=Frame ColorLow, defaultValue=<rgba 0 0 1 1>}
  n2 = VectorParameter() {group=Frame, parameterName=Frame ColorHigh, defaultValue=<rgba 0 0 1 1>}
  n3 = Multiply(A=NamedRerouteUsage(), B=NamedRerouteUsage())
  n4 = Multiply(
   A=Multiply(A=ScalarParameter() {defaultValue=5, group=Frame, parameterName=Frame Brightness}, B=LinearInterpolate(A=@n1.RGB, B=@n2.RGB, Alpha=@n3))
   B=NamedRerouteUsage()
  )
  n5 = Add(
   A=@n4
   B=Multiply(A=NamedRerouteUsage(), B=VectorParameter() {group=Frame, parameterName=Frame Color InnerGlow, defaultValue=<rgba 0 0 1 1>}.RGB)
  )
  LinearInterpolate(A=@n5, Alpha=NamedRerouteUsage())
  n6 = Multiply(
   A=Multiply(A=NamedRerouteUsage(), B=NamedRerouteUsage())
   B=ScalarParameter() {defaultValue=2, group=Flipbook, parameterName=FB_EmissiveStrength}
  )
  n7 = NamedRerouteUsage()
  n8 = VectorParameter() {group=Frame, parameterName=Frame Color Glitches, defaultValue=<rgba 0 1 0.6785 1>}
  n9 = LinearInterpolate(A=Add(A=@n5, B=Multiply(A=@n7, B=@n8.RGB)), B=@n4, Alpha=NamedRerouteUsage())
  MaterialFunctionCall(Emissive=StaticSwitchParameter(True=Add(A=@n6, B=@n9), False=@n6) {group=Frame, parameterName=Enable Frame Border}) {materialFunction=&MF_FX_EmissiveAdjustments}
  n10 = TextureSampleParameter2D!(UVs=NamedRerouteUsage(), "DDX(UVs)"=NamedRerouteUsage(), "DDY(UVs)"=NamedRerouteUsage(), Apply View MipBias) {
   group=Frame
   mipValueMode=TMVM_Derivative
   parameterName=Frame
   samplerSource=SSM_Clamp_WorldGroupSettings
   texture=/Game/Global/Materials/Character_Parent_2023/VideoFlipbook_Parent_2025/MaterialFunctions/T_Cantaloupe_Frame6a
  }
  Multiply(
   A=VectorParameter() {group=Frame, parameterName=Frame Tint, defaultValue=<rgba 0 0 1 1>}.RGB
   B=MaterialFunctionCall(
    LDR Input=Multiply(A=@n10.RGB, B=@n10.RGB)
    HDR Intensity Multiplier=ScalarParameter() {defaultValue=5, group=Frame, parameterName=Frame Brightness}
    Derive HDR Power=ScalarParameter() {defaultValue=2, group=Frame, parameterName=Frame HDR Power}
   ) {materialFunction=&DeriveHDRfromLDR}.RemappedRange
  )
  n11 = Multiply(A=NamedRerouteUsage(), B=ScalarParameter() {defaultValue=1, group=Flipbook, parameterName=FB_GlobalOpacity, sortPriority=201})
  n12 = ScalarParameter() {defaultValue=1, group=Flipbook, parameterName=OpacityAnimation, sortPriority=201}
  NamedRerouteDeclaration(None=Multiply(A=Multiply(A=@n11, B=@n12), B=ScalarParameter() {defaultValue=1, group=Flipbook, parameterName=OpacityMultiplier, sortPriority=201})) {name=Opacity, nodeColor=<rgba 0.2891 0 1 1>}
  NamedRerouteDeclaration(None=StaticBoolParameter() {group=Flipbook, parameterName=FB_DebugFlipbook, sortPriority=101}) {name=DebugFlipbook, nodeColor=<rgba 0.1719 1 0 1>}
  n13 = Subtract(A=TextureCoordinate()) {constB=0.5}
  n14 = Multiply(A=Constant2Vector() {g=1, r=1}, B=ScalarParameter() {defaultValue=0.707, group=Frame, parameterName=Frame DiagonalGradient Size})
  n15 = Power(Base=Abs(None=DotProduct(A=@n13, B=@n14)), Exp=ScalarParameter() {defaultValue=2, group=Frame, parameterName=Frame DiagonalGradient Exp})
  NamedRerouteDeclaration(None=LinearInterpolate(A=ScalarParameter() {group=Frame, parameterName=Frame DiagonalGradient Min}, Alpha=SmoothStep(Value=@n15))) {name=DiagonalGradient, nodeColor=<rgba 0.2891 0 1 1>}
  n16 = ScalarParameter() {defaultValue=0.5, group=Frame, parameterName=Frame Edge Ratio, sliderMax=1}
  n17 = Constant() {r=1}
  n18 = Constant() {r=0.8}
  NamedRerouteDeclaration(None=MaterialFunctionCall(
    UVs=TextureCoordinate()
    Texture Scale=MaterialFunctionCall(Input=@n16, Target Low=@n17, Target High=@n18) {materialFunction=&RemapValueRange}.Result
   ) {materialFunction=&ScaleUVsByCenter}.UVs) {name=FlipbookUVs, nodeColor=<rgba 0.6172 0 1 1>}
  NamedRerouteDeclaration(None=ScalarParameter() {defaultValue=15, group=Flipbook, parameterName=FB_FPS, sortPriority=0}) {name=FPS, nodeColor=<rgba 1 0.8906 0 1>}
  n19 = NamedRerouteDeclaration(None=ScalarParameter() {defaultValue=4, group=Flipbook, parameterName=FB_Columns_01, sortPriority=11}) {name=Columns_01, nodeColor=<rgba 1 0 0.4688 1>}
  n20 = NamedRerouteDeclaration(None=ScalarParameter() {defaultValue=4, group=Flipbook, parameterName=FB_Rows_01, sortPriority=12}) {name=Rows_01, nodeColor=<rgba 1 0 0.4687 1>}
  NamedRerouteDeclaration(None=StaticSwitchParameter(True=ScalarParameter() {group=Flipbook, parameterName=FB_TotalFrames_01, sortPriority=14}, False=Multiply(A=@n19, B=@n20)) {group=Flipbook, parameterName=FB_OverrideTotalFrames_01, sortPriority=13}) {name=TotalFrames_01, nodeColor=<rgba 0.5703 0 1 1>}
  n21 = Time()
  n22 = Max(
   A=ScalarParameter() {group=Flipbook, parameterName=FB_StartTime, sortPriority=100}
   B=ScalarParameter() {group=Flipbook, parameterName=FlipbookStartTime, sortPriority=100}
  )
  n23 = Multiply(A=StaticSwitch(True=@n21, False=Subtract(A=@n21, B=@n22), Value=NamedRerouteUsage()), B=NamedRerouteUsage())
  NamedRerouteDeclaration(None=Fmod(A=NamedRerouteDeclaration(None=@n23) {name=FlipbookTime_01, nodeColor=<rgba 1 0.3281 0 1>}, B=NamedRerouteUsage())) {name=Index_01, nodeColor=<rgba 0 1 0.25 1>}
  NamedRerouteDeclaration(None=MaterialFunctionCall!(
    Flipbook Index=NamedRerouteUsage()
    Flipbook Texture=TextureObjectParameter() {group=Flipbook, parameterName=FB_Texture_01, sortPriority=10, texture=&T_BlacksRGB}
    Flipbook Column Count=NamedRerouteUsage()
    Flipbook Row Count=NamedRerouteUsage()
    UVs=NamedRerouteUsage()
   ) {materialFunction=/Game/Global/Materials/Character_Parent_2023/VideoFlipbook_Parent_2025/MaterialFunctions/MF_FlipbookTexture}.Frame) {name=FlipbookOutput_01, nodeColor=<rgba 1 0 0.0703 1>}
  n24 = NamedRerouteUsage()
  n25 = TextureSample!(UVs=NamedRerouteUsage(), Tex, "DDX(UVs)"=NamedRerouteUsage(), "DDY(UVs)"=NamedRerouteUsage(), Apply View MipBias) {
   mipValueMode=TMVM_Derivative
   samplerSource=SSM_Clamp_WorldGroupSettings
   texture=/Game/Global/Materials/Character_Parent_2023/VideoFlipbook_Parent_2025/MaterialFunctions/T_Cantaloupe_Frame_GlitchMask
  }
  n26 = NamedRerouteDeclaration(None=Power(Base=@n25.R, Exp=ScalarParameter() {defaultValue=1.7, group=Frame, parameterName=Frame_GlitchMaskExp}) {constExponent=1}) {name=Mask_Glitch, nodeColor=<rgba 0.1953 1 0 1>}
  n27 = Time() {bIgnorePause=true}
  n28 = Multiply(A=@n27, B=Constant2Vector() {g=11.17, r=7.123})
  n29 = Frac(None=Multiply(A=Floor(None=@n28), B=Constant2Vector() {g=2.2561, r=1.315}) {constB=1.02})
  n30 = TextureCoordinate() {uTiling=2, vTiling=2}
  n31 = TextureSample!(UVs=Add(A=@n29, B=@n30), Tex, Apply View MipBias) {
   samplerSource=SSM_Wrap_WorldGroupSettings
   samplerType=SAMPLERTYPE_Masks
   texture=/Game/Global/Materials/Character_Parent_2023/VideoFlipbook_Parent_2025/MaterialFunctions/T_Cantaloupe_Rectangle_Glitches
  }
  n32 = TextureSample(UVs=Add(A=@n30, B=MaterialFunctionCall(XY=@n29) {materialFunction=&Swizzle}.YX)) {
   samplerSource=SSM_Wrap_WorldGroupSettings
   samplerType=SAMPLERTYPE_Masks
   texture=&T_Cantaloupe_Rectangle_Glitches
  }
  n33 = TextureSample() {
   samplerSource=SSM_Wrap_WorldGroupSettings
   samplerType=SAMPLERTYPE_Grayscale
   texture=/Game/Global/Textures/Default/Noise/LowResBlurredNoiseNormalized
  }
  n34 = Add(A=@n33.R, B=@n27)
  n35 = Saturate(None=Multiply(A=Multiply(A=@n31.R, B=@n32.G), B=Sine(None=@n34)))
  n36 = Add(A=NamedRerouteUsage(), B=NamedRerouteDeclaration(None=Multiply(A=@n26, B=@n35)) {name=Glitches, nodeColor=<rgba 0 1 0.25 1>})
  NamedRerouteDeclaration(None=Saturate(None=Add(A=@n24, B=@n36))) {name=Frame Opacity, nodeColor=<rgba 0 0.8125 1 1>}
  n37 = NamedRerouteDeclaration(None=ScalarParameter() {defaultValue=4, group=Flipbook, parameterName=FB_Columns_02, sortPriority=22}) {name=Columns_02, nodeColor=<rgba 1 0 0.4688 1>}
  n38 = NamedRerouteDeclaration(None=ScalarParameter() {defaultValue=4, group=Flipbook, parameterName=FB_Rows_02, sortPriority=23}) {name=Rows_02, nodeColor=<rgba 1 0 0.4688 1>}
  NamedRerouteDeclaration(None=StaticSwitchParameter(True=ScalarParameter() {group=Flipbook, parameterName=FB_TotalFrames_02, sortPriority=25}, False=Multiply(A=@n37, B=@n38)) {group=Flipbook, parameterName=FB_OverrideTotalFrames_02, sortPriority=24}) {name=TotalFrames_02, nodeColor=<rgba 0.5703 0 1 1>}
  n39 = NamedRerouteUsage()
  n40 = Add(A=@n39, B=NamedRerouteUsage())
  n41 = NamedRerouteUsage()
  n42 = NamedRerouteUsage()
  NamedRerouteDeclaration(None=StaticSwitch(True=Add(A=@n40, B=@n41), False=StaticSwitch(True=@n40, False=@n39, Value=@n42), Value=NamedRerouteUsage())) {name=TotalFrames, nodeColor=<rgba 0 1 0.6484 1>}
  n43 = Subtract(A=NamedRerouteUsage(), B=NamedRerouteUsage())
  NamedRerouteDeclaration(None=Fmod(A=NamedRerouteDeclaration(None=@n43) {name=FlipbookTime_02, nodeColor=<rgba 1 0.2344 0 1>}, B=NamedRerouteUsage())) {name=Index_02, nodeColor=<rgba 0 1 0.25 1>}
  NamedRerouteDeclaration(None=MaterialFunctionCall(
    Flipbook Index=NamedRerouteUsage()
    Flipbook Texture=TextureObjectParameter() {group=Flipbook, parameterName=FB_Texture_02, sortPriority=21, texture=&T_BlacksRGB}
    Flipbook Column Count=NamedRerouteUsage()
    Flipbook Row Count=NamedRerouteUsage()
    UVs=NamedRerouteUsage()
   ) {materialFunction=&MF_FlipbookTexture}.Frame) {name=FlipbookOutput_02, nodeColor=<rgba 1 0 0.0703 1>}
  NamedRerouteDeclaration(None=Step(Y=NamedRerouteUsage(), X=NamedRerouteUsage())) {name=Blend_01, nodeColor=<rgba 1 0 0.6094 1>}
  n44 = NamedRerouteDeclaration(None=ScalarParameter() {defaultValue=4, group=Flipbook, parameterName=FB_Columns_03}) {name=Columns_03, nodeColor=<rgba 1 0 0.4688 1>}
  n45 = NamedRerouteDeclaration(None=ScalarParameter() {defaultValue=4, group=Flipbook, parameterName=FB_Rows_03, sortPriority=33}) {name=Rows_03, nodeColor=<rgba 1 0 0.4688 1>}
  NamedRerouteDeclaration(None=StaticSwitchParameter(True=ScalarParameter() {group=Flipbook, parameterName=FB_TotalFrames_03, sortPriority=35}, False=Multiply(A=@n44, B=@n45)) {group=Flipbook, parameterName=FB_OverrideTotalFrames_03, sortPriority=34}) {name=TotalFrames_03, nodeColor=<rgba 0.5703 0 1 1>}
  n46 = ScalarParameter() {defaultValue=0.5, group=Frame, parameterName=Frame Edge Ratio, sliderMax=1}
  NamedRerouteDeclaration(None=Divide(A=MaterialFunctionCall!(UVs, EdgeRatio=@n46, "Texture (even slices 3x3)") {materialFunction=/Game/Global/Materials/Character_Parent_2023/VideoFlipbook_Parent_2025/MaterialFunctions/MF_NineSliceUVs}.Deformed UV) {constB=3}) {name=Frame UV, nodeColor=<rgba 0 1 0.7656 1>}
  n47 = NamedRerouteUsage()
  n48 = NamedRerouteUsage()
  NamedRerouteDeclaration(None=Fmod(A=Subtract(A=@n47, B=@n48), B=NamedRerouteUsage())) {name=Index_03, nodeColor=<rgba 0 1 0.25 1>}
  NamedRerouteDeclaration(None=MaterialFunctionCall(
    Flipbook Index=NamedRerouteUsage()
    Flipbook Texture=TextureObjectParameter() {group=Flipbook, parameterName=FB_Texture_03, sortPriority=31, texture=&T_BlacksRGB}
    Flipbook Column Count=NamedRerouteUsage()
    Flipbook Row Count=NamedRerouteUsage()
    UVs=NamedRerouteUsage()
   ) {materialFunction=&MF_FlipbookTexture}.Frame) {name=FlipbookOutput_03, nodeColor=<rgba 1 0 0.0703 1>}
  n49 = NamedRerouteUsage()
  n50 = NamedRerouteUsage()
  NamedRerouteDeclaration(None=Step(Y=Add(A=@n49, B=@n50), X=NamedRerouteUsage())) {name=Blend_02, nodeColor=<rgba 0.8516 1 0 1>}
  NamedRerouteDeclaration(None=StaticBoolParameter() {group=Flipbook, parameterName=Use2Flipbooks, sortPriority=50}) {name=Use2Flipbooks, nodeColor=<rgba 0 1 0.9531 1>}
  NamedRerouteDeclaration(None=StaticBoolParameter() {group=Flipbook, parameterName=Use3Flipbooks, sortPriority=51}) {name=Use3Flipbooks, nodeColor=<rgba 0 1 0.9531 1>}
  n51 = NamedRerouteUsage()
  n52 = NamedRerouteUsage()
  n53 = LinearInterpolate(A=@n52, B=NamedRerouteUsage(), Alpha=NamedRerouteUsage())
  n54 = NamedRerouteUsage()
  n55 = NamedRerouteUsage()
  n56 = NamedRerouteUsage()
  n57 = Constant()
  n58 = If(
   A=@n51
   "A > B"=StaticSwitch(True=LinearInterpolate(A=@n53, B=@n54, Alpha=@n55), False=StaticSwitch(True=@n53, False=@n52, Value=@n56), Value=NamedRerouteUsage())
   "A == B"=@n57
   "A < B"=@n57
  )
  n59 = NamedRerouteUsage()
  NamedRerouteDeclaration(None=StaticSwitch(True=@n58, False=If(A=@n51, B=@n59, "A > B"=@n57, "A == B"=@n58, "A < B"=@n58), Value=NamedRerouteUsage())) {name=Flipbook_OUT, nodeColor=<rgba 0.3125 0 1 1>}
  n60 = TextureCoordinate()
  NamedRerouteDeclaration(None=DDX(Value=@n60)) {name=Frame DDX, nodeColor=<rgba 0 1 0.7656 1>}
  NamedRerouteDeclaration(None=DDY(Value=@n60)) {name=Frame DDY, nodeColor=<rgba 0 1 0.7656 1>}
  n61 = TextureSample!(UVs=NamedRerouteUsage(), Tex, "DDX(UVs)"=NamedRerouteUsage(), "DDY(UVs)"=NamedRerouteUsage(), Apply View MipBias) {
   mipValueMode=TMVM_Derivative
   samplerSource=SSM_Clamp_WorldGroupSettings
   texture=/Game/Global/Materials/Character_Parent_2023/VideoFlipbook_Parent_2025/MaterialFunctions/T_Cantaloupe_Frame_Content
  }
  NamedRerouteDeclaration(None=@n61.R) {name=Mask_Content, nodeColor=<rgba 0.1953 1 0 1>}
  n62 = TextureSample(UVs=NamedRerouteUsage(), "DDX(UVs)"=NamedRerouteUsage(), "DDY(UVs)"=NamedRerouteUsage()) {
   mipValueMode=TMVM_Derivative
   samplerSource=SSM_Clamp_WorldGroupSettings
   texture=/Game/Global/Materials/Character_Parent_2023/VideoFlipbook_Parent_2025/MaterialFunctions/T_Cantaloupe_Frame_Frame
  }
  NamedRerouteDeclaration(None=@n62.R) {name=Mask_Frame, nodeColor=<rgba 0.1953 1 0 1>}
  n63 = TextureSample(UVs=NamedRerouteUsage(), "DDX(UVs)"=NamedRerouteUsage(), "DDY(UVs)"=NamedRerouteUsage()) {
   mipValueMode=TMVM_Derivative
   samplerSource=SSM_Clamp_WorldGroupSettings
   texture=/Game/Global/Materials/Character_Parent_2023/VideoFlipbook_Parent_2025/MaterialFunctions/T_Cantaloupe_Frame_InnerGlow
  }
  NamedRerouteDeclaration(None=@n63.R) {name=Mask_InnerGlow, nodeColor=<rgba 0.1953 1 0 1>}
  NamedRerouteUsage()
  n64 = ComponentMask(None=TextureCoordinate()) {r=true}
  SmoothStep(
   Max=ScalarParameter() {defaultValue=0.8, group=Flipbook, parameterName=FB_OpacityEdgeFade, sortPriority=200}
   Value=Abs(None=ConstantBiasScale(None=@n64) {bias=-0.5, scale=2})
  ) {constMax=0.9, constMin=1}
 ]
}
